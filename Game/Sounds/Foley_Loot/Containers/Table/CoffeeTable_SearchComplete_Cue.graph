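# Sound cue for the foley played when a coffee table container search completes, using loot-collect sound class and prop-search attenuation.
# A quality-level switch selects a random pick from three search waves, a pick from two, or, for the last two slots, one shared optimized search-end wave.
# All branches vary pitch and volume slightly, and the random branches also fade out through envelopers.
# A concurrency override caps simultaneous instances, stopping the farthest first.
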SoundCue /Game/Sounds/Foley_Loot/Containers/Table/CoffeeTable_SearchComplete_Cue
root=QualityLevel_0
nodes [
 {
  id=Enveloper_0
  children [Random_0]
  properties {pitchCurve {editorCurveData {keys [{time=0}]}}, volumeCurve {editorCurveData {keys [<at 0 1>,<at 0.65 0>]}}}
 }
 {
  id=Enveloper_2
  children [Random_46]
  properties {pitchCurve {editorCurveData {keys [{time=0}]}}, volumeCurve {editorCurveData {keys [<at 0 1>,<at 0.65 0>]}}}
 }
 {id=Modulator_0, children [Enveloper_0], properties {pitchMax=1.07, pitchMin=0.92, volumeMax=1.07, volumeMin=0.92}}
 {id=Modulator_1, children [WavePlayer_0], properties {pitchMax=1.07, pitchMin=0.92, volumeMax=1.07, volumeMin=0.92}}
 {id=Modulator_5, children [Enveloper_2], properties {pitchMax=1.07, pitchMin=0.92, volumeMax=1.07, volumeMin=0.92}}
 {id=QualityLevel_0, children [Modulator_5,Modulator_0,Modulator_1,Modulator_1]}
 {id=Random_0, children [WavePlayer_2,WavePlayer_3], properties {weights [1,1]}}
 {id=Random_46, children [WavePlayer_290, WavePlayer_293, WavePlayer_294], properties {weights [1,1,1]}}
 {id=WavePlayer_0, properties {soundWaveAssetPtr=&/CoffeeTable_SearchEnd_opt}}
 {id=WavePlayer_2, properties {soundWaveAssetPtr=&/CoffeeTable_Search03}}
 {id=WavePlayer_290, properties {soundWaveAssetPtr=&CoffeeTable_Search03}}
 {id=WavePlayer_293, properties {soundWaveAssetPtr=&/CoffeeTable_Search09}}
 {id=WavePlayer_294, properties {soundWaveAssetPtr=&/CoffeeTable_Search10}}
 {id=WavePlayer_3, properties {soundWaveAssetPtr=&CoffeeTable_Search09}}
]
properties {
 attenuationSettings=/Game/Sounds/Attenuation/Fort_Prop_Search_Att
 bOverrideConcurrency=true
 maxDistance=1300
 soundClassObject=&Fort_Loot_Collect
 volumeMultiplier=0.7
 concurrencyOverrides {maxCount=4, resolutionRule=StopFarthestThenPreventNew}
}
