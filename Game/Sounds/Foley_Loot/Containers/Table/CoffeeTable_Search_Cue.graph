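# Sound cue for the foley played when searching a coffee table loot container, with prop-search attenuation and prevent-new concurrency.
# A quality-level root picks between two layered mixes, with differing variant counts, and two branches playing a single optimized wave.
# Each mix combines a randomized table search, a one-second-delayed replay through the same modulator, and a quieter random gear rustle.
# Modulators slightly vary pitch and volume on the table sounds.
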SoundCue /Game/Sounds/Foley_Loot/Containers/Table/CoffeeTable_Search_Cue
root=QualityLevel_0
nodes [
 {id=Delay_0, children [Modulator_0], properties {delayMax=1, delayMin=1}}
 {id=Delay_1, children [Modulator_2], properties {delayMax=1, delayMin=1}}
 {id=Mixer_0, children [Modulator_0,Delay_0,Random_1], properties {inputVolume [1,0.8,0.5]}}
 {id=Mixer_3, children [Modulator_2,Delay_1,Random_5], properties {inputVolume [1,0.8,0.5]}}
 {id=Modulator_0, children [Random_0], properties {pitchMax=1.07, pitchMin=0.92, volumeMax=1.07, volumeMin=0.92}}
 {id=Modulator_1, children [WavePlayer_0], properties {pitchMax=1.07, pitchMin=0.92, volumeMax=1.07, volumeMin=0.92}}
 {id=Modulator_2, children [Random_47], properties {pitchMax=1.07, pitchMin=0.92, volumeMax=1.07, volumeMin=0.92}}
 {id=Modulator_3, children [WavePlayer_7], properties {pitchMax=1.07, pitchMin=0.92, volumeMax=1.07, volumeMin=0.92}}
 {id=QualityLevel_0, children [Mixer_3,Mixer_0,Modulator_1,Modulator_3]}
 {id=Random_0, children [WavePlayer_1,WavePlayer_2], properties {weights [1,1]}}
 {id=Random_1, children [WavePlayer_4,WavePlayer_5], properties {weights [1,1]}}
 {id=Random_47, children [WavePlayer_297, WavePlayer_300, WavePlayer_301], properties {weights [1,1,1]}}
 {id=Random_5, children [WavePlayer_12, WavePlayer_13, WavePlayer_14], properties {weights [1,1,1]}}
 {id=WavePlayer_0, properties {soundWaveAssetPtr=&/CoffeeTable_Search_opt}}
 {id=WavePlayer_1, properties {soundWaveAssetPtr=&/CoffeeTable_Search03}}
 {id=WavePlayer_12, properties {soundWaveAssetPtr=/Game/Sounds/Foley_Loot/Containers/Misc/Prop_Gear_Search_05}}
 {id=WavePlayer_13, properties {soundWaveAssetPtr=/Game/Sounds/Foley_Loot/Containers/Misc/Prop_Gear_Search_06}}
 {id=WavePlayer_14, properties {soundWaveAssetPtr=/Game/Sounds/Foley_Loot/Containers/Misc/Prop_Gear_Search_01}}
 {id=WavePlayer_2, properties {soundWaveAssetPtr=&/CoffeeTable_Search09}}
 {id=WavePlayer_297, properties {soundWaveAssetPtr=&CoffeeTable_Search03}}
 {id=WavePlayer_300, properties {soundWaveAssetPtr=&CoffeeTable_Search09}}
 {id=WavePlayer_301, properties {soundWaveAssetPtr=&/CoffeeTable_Search10}}
 {id=WavePlayer_4, properties {soundWaveAssetPtr=&Prop_Gear_Search_05}}
 {id=WavePlayer_5, properties {soundWaveAssetPtr=&Prop_Gear_Search_06}}
 {id=WavePlayer_7, properties {soundWaveAssetPtr=&CoffeeTable_Search_opt}}
]
properties {
 attenuationSettings=/Game/Sounds/Attenuation/Fort_Prop_Search_Att
 bOverrideConcurrency=true
 maxDistance=1300
 soundClassObject=&Fort_Loot_Collect
 volumeMultiplier=1
 concurrencyOverrides {resolutionRule=PreventNew}
}
