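# Sound cue for completing a search of a kitchen table loot container.
# A random pick between two kitchen table search waves runs through an enveloper that fades volume out over under a second, then a modulator adds slight pitch and volume variation.
# It uses prop-search attenuation and the loot collect sound class, with a concurrency override that caps instances and stops the farthest first.
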SoundCue /Game/Sounds/Foley_Loot/Containers/Table/KitchenTable_SearchComplete_Cue
root=Modulator_5
nodes [
 {
  id=Enveloper_12
  children [Random_15]
  properties {pitchCurve {editorCurveData {keys [{time=0}]}}, volumeCurve {editorCurveData {keys [<at 0 1>,<at 0.7 0>]}}}
 }
 {id=Modulator_5, children [Enveloper_12], properties {pitchMax=1.07, pitchMin=0.92, volumeMax=1.07, volumeMin=0.92}}
 {id=Random_15, children [WavePlayer_113, WavePlayer_114], properties {weights [1,1]}}
 {id=WavePlayer_113, properties {soundWaveAssetPtr=&/KitchenTable_Search01}}
 {id=WavePlayer_114, properties {soundWaveAssetPtr=&/KitchenTable_Search02}}
]
properties {
 attenuationSettings=/Game/Sounds/Attenuation/Fort_Prop_Search_Att
 bOverrideConcurrency=true
 maxDistance=1300
 soundClassObject=&Fort_Loot_Collect
 volumeMultiplier=0.77
 concurrencyOverrides {maxCount=4, resolutionRule=StopFarthestThenPreventNew}
}
