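# KitchenTable_Search_Cue is the loot foley sound cue for searching a kitchen table container.
# A mixer plays one shared branch twice, immediately and through a one-second delay, layering two search sounds.
# That branch randomly picks one of three equally weighted waves and slightly modulates pitch and volume.
# It uses prop-search attenuation, the loot-collect sound class and a concurrency override that prevents new instances at the limit.
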SoundCue /Game/Sounds/Foley_Loot/Containers/Table/KitchenTable_Search_Cue
root=Mixer_11
nodes [
 {id=Delay_28, children [Modulator_2], properties {delayMax=1, delayMin=1}}
 {id=Mixer_11, children [Modulator_2,Delay_28], properties {inputVolume [1,1]}}
 {id=Modulator_2, children [Random_16], properties {pitchMax=1.07, pitchMin=0.92, volumeMax=1.07, volumeMin=0.92}}
 {id=Random_16, children [WavePlayer_116, WavePlayer_117, WavePlayer_118], properties {weights [1,1,1]}}
 {id=WavePlayer_116, properties {soundWaveAssetPtr=&/KitchenTable_Search04}}
 {id=WavePlayer_117, properties {soundWaveAssetPtr=&/KitchenTable_Search01}}
 {id=WavePlayer_118, properties {soundWaveAssetPtr=&/KitchenTable_Search02}}
]
properties {
 attenuationSettings=/Game/Sounds/Attenuation/Fort_Prop_Search_Att
 bOverrideConcurrency=true
 maxDistance=1300
 soundClassObject=&Fort_Loot_Collect
 volumeMultiplier=0.85
 concurrencyOverrides {resolutionRule=PreventNew}
}
